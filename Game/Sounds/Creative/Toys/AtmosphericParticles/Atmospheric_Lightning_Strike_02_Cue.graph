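# Sound cue for a lightning strike variant, apparently for the Creative atmospheric particle toys, played under the storms sound class.
# Mixes three layers under a root modulator: storm lightning spawn, a briefly delayed half-volume Zapper impact, and a later Zapper shot.
# Spawn and shot layers are pitched up and faded by matching volume envelopes.
# The shot layer gets its own far attenuation; the cue uses medium attenuation and a tight concurrency cap.
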SoundCue /Game/Sounds/Creative/Toys/AtmosphericParticles/Atmospheric_Lightning_Strike_02_Cue
root=Modulator_0
nodes [
 {id=Attenuation_0, children [Enveloper_2], properties {attenuationSettings=/Game/Sounds/Creative/Attenuations/Fort_Creative_FarToy_Att}}
 {id=Delay_0, children [WavePlayer_3], properties {delayMax=0.04, delayMin=0.04}}
 {id=Delay_1, children [Modulator_2], properties {delayMax=0.08, delayMin=0.08}}
 {
  id=Enveloper_0
  children [Modulator_1]
  properties {pitchCurve {editorCurveData {keys [{time=0}]}}, volumeCurve {editorCurveData {keys [<at 0 1>,<at 0.8 0.9>,<at 1.2 0>]}}}
 }
 {
  id=Enveloper_2
  children [Delay_1]
  properties {pitchCurve {editorCurveData {keys [{time=0}]}}, volumeCurve {editorCurveData {keys [<at 0 1>,<at 0.8 0.9>,<at 1.2 0>]}}}
 }
 {id=Mixer_0, children [Enveloper_0, Delay_0, Attenuation_0], properties {inputVolume [1,0.5,1]}}
 {id=Modulator_0, children [Mixer_0]}
 {id=Modulator_1, children [WavePlayer_4], properties {pitchMax=1.2, pitchMin=1.2}}
 {id=Modulator_2, children [WavePlayer_5], properties {pitchMax=1.2, pitchMin=1.2}}
 {id=WavePlayer_3, properties {soundWaveAssetPtr=/Game/Sounds/Fort_Monster_Zapper/Zapper_Weapon_Impact_Lightning_04}}
 {id=WavePlayer_4, properties {soundWaveAssetPtr=/Game/Sounds/Fort_GamePlay_Sounds/Storms/lightning_spawn_03}}
 {id=WavePlayer_5, properties {soundWaveAssetPtr=/Game/Sounds/Fort_Monster_Zapper/Zapper_Weapon_Shot_Lightning_01}}
]
properties {
 attenuationSettings=/Game/Sounds/Creative/Attenuations/Fort_Creative_MediumToy_Att
 bOverrideConcurrency=true
 maxDistance=20000
 soundClassObject=/Game/Sounds/FortSoundClassesAndModes/Fort_Storms_PSM
 volumeMultiplier=1.4
 concurrencyOverrides {maxCount=2}
}
