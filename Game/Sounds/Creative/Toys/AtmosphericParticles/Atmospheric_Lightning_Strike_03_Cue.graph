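# Sound cue for a Creative atmospheric particle toy's lightning strike, reusing Storm King and crack lightning waves.
# A root modulator plays four mixed layers: a fixed impact, a delayed random crack impact, a random projectile-appear sound, and a distant-layer submix.
# That submix uses its own far-toy attenuation for a delayed pitched-up charge and an envelope-faded impact.
# The cue defaults to medium-toy attenuation and overrides concurrency to limit simultaneous plays.
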SoundCue /Game/Sounds/Creative/Toys/AtmosphericParticles/Atmospheric_Lightning_Strike_03_Cue
root=Modulator_0
nodes [
 {id=Attenuation_0, children [Mixer_1], properties {attenuationSettings=/Game/Sounds/Creative/Attenuations/Fort_Creative_FarToy_Att}}
 {id=Delay_0, children [Random_1], properties {delayMax=0.04, delayMin=0.04}}
 {id=Delay_1, children [Modulator_2], properties {delayMax=0.08, delayMin=0.08}}
 {id=Delay_2, children [WavePlayer_16], properties {delayMax=0.08, delayMin=0.08}}
 {
  id=Enveloper_1
  children [Delay_2]
  properties {pitchCurve {editorCurveData {keys [{time=0}]}}, volumeCurve {editorCurveData {keys [<at 0 1>,<at 2 0.9>,<at 2.2777 0>]}}}
 }
 {id=Mixer_0, children [WavePlayer_7, Delay_0, Random_0, Attenuation_0], properties {inputVolume [0.7,1,0.7,1]}}
 {id=Mixer_1, children [Delay_1,Enveloper_1], properties {inputVolume [1,0.6]}}
 {id=Modulator_0, children [Mixer_0]}
 {id=Modulator_2, children [WavePlayer_10], properties {pitchMax=1.2, pitchMin=1.2}}
 {id=Random_0, children [WavePlayer_0,WavePlayer_1], properties {weights [1,1]}}
 {id=Random_1, children [WavePlayer_13, WavePlayer_14], properties {weights [1,1]}}
 {id=WavePlayer_0, properties {soundWaveAssetPtr=/Game/Sounds/Fort_Monster_Storm_King/SFX/Final_SFX/StormKing_Projectile_Lightning_Appear_01}}
 {id=WavePlayer_1, properties {soundWaveAssetPtr=/Game/Sounds/Fort_Monster_Storm_King/SFX/Final_SFX/StormKing_Projectile_Lightning_Appear_02}}
 {
  id=WavePlayer_10
  properties {soundWaveAssetPtr=/Game/Sounds/Fort_Monster_Storm_King/SFX/Final_SFX/StormKing_LightningSummons_Charge_Right_02}
 }
 {id=WavePlayer_13, properties {soundWaveAssetPtr=/Game/Athena/Sounds/Events/CrackLightning/Crack_Lightning_StrikeImpact_01}}
 {id=WavePlayer_14, properties {soundWaveAssetPtr=/Game/Athena/Sounds/Events/CrackLightning/Crack_Lightning_StrikeImpact_02}}
 {
  id=WavePlayer_16
  properties {soundWaveAssetPtr=/Game/Sounds/Fort_Monster_Storm_King/SFX/Final_SFX/StormKing_LightningSummons_Lightning_Impact_03}
 }
 {
  id=WavePlayer_7
  properties {soundWaveAssetPtr=/Game/Sounds/Fort_Monster_Storm_King/SFX/Final_SFX/StormKing_LightningSummons_Lightning_Impact_01}
 }
]
properties {
 attenuationSettings=/Game/Sounds/Creative/Attenuations/Fort_Creative_MediumToy_Att
 bOverrideConcurrency=true
 maxDistance=20000
 soundClassObject=/Game/Sounds/FortSoundClassesAndModes/Fort_Storms_PSM
 volumeMultiplier=1.4
 concurrencyOverrides {maxCount=2}
}
